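# Control Rig function library providing DeformRig_v1_Twists, which drives limb twist bones on a biped through bone-item inputs.
# Upper arm and thigh twists counter-rotate using scaled swing-twist X extraction; thighs measure against initial pose.
# Upperarm_twist_02 reads only the initial upper arm pose, unlike twist_01.
# Forearm and calf twist_01 bones follow hand or foot roll through weighted parent constraints, keep only X, and pass half to twist_02.
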ControlRigBlueprint /BRCosmetics/Animation/Libraries/DeformRig_v1_FunctionLibrary
graphs [
 New Function_ContainedGraph {nodes [
   RigVMFunctionEntryNode()
   n1 = RigVMVariableNode() {variable=upperarm_l}
   n2 = /Script/RigVM.RigVMFunction_MathQuaternionToAxisAndAngle(Value=&RigVMFunction_MathQuaternionSwingTwist(TwistAxis=(X=1,Y=0,Z=0), Input=&RigUnit_GetTransform(Space=LocalSpace, bInitial=False, Item=@n1.Value).Transform.Rotation).Twist) {is=struct RigVMUnitNode, template="ToAxisAndAngle::Execute(in Value,out Axis,out Angle)"}
   RerouteNode_2_1 = RigVMDispatchNode(Value=-0.8) {template="DISPATCH_RigVMDispatch_Constant(io Value)"}
   > Set Transform_4_1_1_2 = &RigUnit_SetRotation(
    Weight=1
    bInitial=False
    Value=/Script/RigVM.RigVMFunction_MathQuaternionFromAxisAndAngle(
     Angle=&RigVMFunction_MathFloatMul(A=@n2.Angle, B=@RerouteNode_2_1.Value) {template="Multiply::Execute(in A,in B,out Result)"}.Result
     Axis=@n2.Axis
    ) {is=struct RigVMUnitNode, template="FromAxisAndAngle::Execute(in Axis,in Angle,out Result)"}.Result
    Item=RigVMVariableNode() {variable=upperarm_twist_01_l}.Value
   )
   n3 = RigVMVariableNode() {variable=upperarm_l}
   n4 = &RigVMFunction_MathQuaternionToAxisAndAngle(Value=&RigVMFunction_MathQuaternionSwingTwist(TwistAxis=(X=1,Y=0,Z=0), Input=&RigUnit_GetTransform(Space=LocalSpace, bInitial=true, Item=@n3.Value).Transform.Rotation).Twist) {template="ToAxisAndAngle::Execute(in Value,out Axis,out Angle)"}
   RerouteNode_2 = RigVMDispatchNode(Value=-0.5) {template="DISPATCH_RigVMDispatch_Constant(io Value)"}
   > Set Transform_4_1_1_1_1 = &RigUnit_SetRotation(
    Weight=1
    bInitial=False
    Value=&RigVMFunction_MathQuaternionFromAxisAndAngle(
     Angle=&RigVMFunction_MathFloatMul(A=@n4.Angle, B=@RerouteNode_2.Value) {template="Multiply::Execute(in A,in B,out Result)"}.Result
     Axis=@n4.Axis
    ) {template="FromAxisAndAngle::Execute(in Axis,in Angle,out Result)"}.Result
    Item=RigVMVariableNode() {variable=upperarm_twist_02_l}.Value
   )
   n5 = RigVMVariableNode() {variable=upperarm_r}
   n6 = &RigVMFunction_MathQuaternionToAxisAndAngle(Value=&RigVMFunction_MathQuaternionSwingTwist(TwistAxis=(X=1,Y=0,Z=0), Input=&RigUnit_GetTransform(Space=LocalSpace, bInitial=False, Item=@n5.Value).Transform.Rotation).Twist) {template="ToAxisAndAngle::Execute(in Value,out Axis,out Angle)"}
   > Set Transform_4_1_1 = &RigUnit_SetRotation(
    Weight=1
    bInitial=False
    Value=&RigVMFunction_MathQuaternionFromAxisAndAngle(
     Angle=&RigVMFunction_MathFloatMul(A=@n6.Angle, B=@RerouteNode_2_1.Value) {template="Multiply::Execute(in A,in B,out Result)"}.Result
     Axis=@n6.Axis
    ) {template="FromAxisAndAngle::Execute(in Axis,in Angle,out Result)"}.Result
    Item=RigVMVariableNode() {variable=upperarm_twist_01_r}.Value
   )
   n7 = RigVMVariableNode() {variable=upperarm_r}
   n8 = &RigVMFunction_MathQuaternionToAxisAndAngle(Value=&RigVMFunction_MathQuaternionSwingTwist(TwistAxis=(X=1,Y=0,Z=0), Input=&RigUnit_GetTransform(Space=LocalSpace, bInitial=true, Item=@n7.Value).Transform.Rotation).Twist) {template="ToAxisAndAngle::Execute(in Value,out Axis,out Angle)"}
   > Set Transform_4_1_1_1 = &RigUnit_SetRotation(
    Weight=1
    bInitial=False
    Value=&RigVMFunction_MathQuaternionFromAxisAndAngle(
     Angle=&RigVMFunction_MathFloatMul(A=@n8.Angle, B=@RerouteNode_2.Value) {template="Multiply::Execute(in A,in B,out Result)"}.Result
     Axis=@n8.Axis
    ) {template="FromAxisAndAngle::Execute(in Axis,in Angle,out Result)"}.Result
    Item=RigVMVariableNode() {variable=upperarm_twist_02_r}.Value
   )
   n9 = &RigUnit_GetTransform(Space=LocalSpace, bInitial=False, Item=RigVMVariableNode() {variable=thigh_l}.Value)
   n10 = &RigUnit_GetTransform(Space=LocalSpace, bInitial=true, Item=RigVMVariableNode() {variable=thigh_l}.Value)
   n11 = &RigVMFunction_MathQuaternionToAxisAndAngle(Value=&RigVMFunction_MathQuaternionSwingTwist(
     TwistAxis=(X=1,Y=0,Z=0)
     Input=&RigVMFunction_MathTransformMakeRelative(Global=@n9.Transform, Parent=@n10.Transform).Local.Rotation
    ).Twist) {template="ToAxisAndAngle::Execute(in Value,out Axis,out Angle)"}
   > Set Transform_4_1 = &RigUnit_SetRotation(
    Weight=1
    bInitial=False
    Value=&RigVMFunction_MathQuaternionFromAxisAndAngle(
     Angle=&RigVMFunction_MathFloatMul(B=-1, A=@n11.Angle) {template="Multiply::Execute(in A,in B,out Result)"}.Result
     Axis=@n11.Axis
    ) {template="FromAxisAndAngle::Execute(in Axis,in Angle,out Result)"}.Result
    Item=RigVMVariableNode() {variable=thigh_twist_01_l}.Value
   )
   n12 = &RigUnit_GetTransform(Space=LocalSpace, bInitial=true, Item=RigVMVariableNode() {variable=thigh_r}.Value)
   n13 = &RigUnit_GetTransform(Space=LocalSpace, bInitial=False, Item=RigVMVariableNode() {variable=thigh_r}.Value)
   n14 = &RigVMFunction_MathQuaternionToAxisAndAngle(Value=&RigVMFunction_MathQuaternionSwingTwist(
     TwistAxis=(X=1,Y=0,Z=0)
     Input=&RigVMFunction_MathTransformMakeRelative(Parent=@n12.Transform, Global=@n13.Transform).Local.Rotation
    ).Twist) {template="ToAxisAndAngle::Execute(in Value,out Axis,out Angle)"}
   > Set Transform_4 = &RigUnit_SetRotation(
    Weight=1
    bInitial=False
    Value=&RigVMFunction_MathQuaternionFromAxisAndAngle(
     Angle=&RigVMFunction_MathFloatMul(B=-1, A=@n14.Angle) {template="Multiply::Execute(in A,in B,out Result)"}.Result
     Axis=@n14.Axis
    ) {template="FromAxisAndAngle::Execute(in Axis,in Angle,out Result)"}.Result
    Item=RigVMVariableNode() {variable=thigh_twist_01_r}.Value
   )
   > &RigUnit_ParentConstraint(
    AdvancedSettings=(InterpolationType=Average,RotationOrderForFilter=XZY)
    Filter=(TranslationFilter=(bX=false,bY=false,bZ=false),RotationFilter=(bX=True,bY=false,bZ=false),ScaleFilter=(bX=false,bY=false,bZ=false))
    Parents=((Item=(Type=Bone,Name="hand_l"),Weight=0.75),(Item=(Type=Bone,Name="lowerarm_twist_01_l"),Weight=0.25))
    Weight=1
    Child=RigVMVariableNode() {variable=lowerarm_twist_01_l}.Value
    Parents.0.Item=RigVMVariableNode() {variable=hand_l}.Value
    Parents.1.Item=RigVMVariableNode() {variable=lowerarm_twist_01_l}.Value
   )
   n15 = &RigVMFunction_MathQuaternionToEuler(
    RotationOrder=ZYX
    Value=&RigUnit_GetTransform(Space=LocalSpace, bInitial=False, Item=RigVMVariableNode() {variable=lowerarm_twist_01_l}.Value).Transform.Rotation
   )
   > Set Transform_3 = &RigUnit_SetRotation(
    Weight=1
    bInitial=False
    Value=&RigVMFunction_MathQuaternionFromEuler(Euler=(X=0,Y=0,Z=0), Euler.X=@n15.Result.X).Result
    Item=RigVMVariableNode() {variable=lowerarm_twist_01_l}.Value
   )
   > Set Transform_1_1 = &RigUnit_SetRotation(
    Weight=1
    bInitial=False
    Value=&RigVMFunction_MathQuaternionFromEuler(
     Euler=(X=0,Y=0,Z=0)
     Euler.X=&RigVMFunction_MathFloatMul(B=0.5, A=@n15.Result.X) {template="Multiply::Execute(in A,in B,out Result)"}.Result
    ).Result
    Item=RigVMVariableNode() {variable=lowerarm_twist_02_l}.Value
   )
   > &RigUnit_ParentConstraint(
    AdvancedSettings=(InterpolationType=Average,RotationOrderForFilter=XZY)
    Filter=(TranslationFilter=(bX=false,bY=false,bZ=false),RotationFilter=(bX=True,bY=false,bZ=false),ScaleFilter=(bX=false,bY=false,bZ=false))
    Parents=((Item=(Type=Bone,Name="hand_r"),Weight=0.75),(Item=(Type=Bone,Name="lowerarm_twist_01_r"),Weight=0.25))
    Weight=1
    Child=RigVMVariableNode() {variable=lowerarm_twist_01_r}.Value
    Parents.0.Item=RigVMVariableNode() {variable=hand_r}.Value
    Parents.1.Item=RigVMVariableNode() {variable=lowerarm_twist_01_r}.Value
   )
   n16 = &RigVMFunction_MathQuaternionToEuler(
    RotationOrder=ZYX
    Value=&RigUnit_GetTransform(Space=LocalSpace, bInitial=False, Item=RigVMVariableNode() {variable=lowerarm_twist_01_r}.Value).Transform.Rotation
   )
   > Set Transform_2_2 = &RigUnit_SetRotation(
    Weight=1
    bInitial=False
    Value=&RigVMFunction_MathQuaternionFromEuler(Euler=(X=0,Y=0,Z=0), Euler.X=@n16.Result.X).Result
    Item=RigVMVariableNode() {variable=lowerarm_twist_01_r}.Value
   )
   > Set Transform_2_1_1 = &RigUnit_SetRotation(
    Weight=1
    bInitial=False
    Value=&RigVMFunction_MathQuaternionFromEuler(
     Euler=(X=0,Y=0,Z=0)
     Euler.X=&RigVMFunction_MathFloatMul(B=0.5, A=@n16.Result.X) {template="Multiply::Execute(in A,in B,out Result)"}.Result
    ).Result
    Item=RigVMVariableNode() {variable=lowerarm_twist_02_r}.Value
   )
   > &RigUnit_ParentConstraint(
    AdvancedSettings=(InterpolationType=Average,RotationOrderForFilter=XZY)
    Filter=(TranslationFilter=(bX=false,bY=false,bZ=false),RotationFilter=(bX=True,bY=false,bZ=false),ScaleFilter=(bX=false,bY=false,bZ=false))
    Parents=((Item=(Type=Bone,Name="foot_l"),Weight=0.75),(Item=(Type=Bone,Name="calf_twist_01_l"),Weight=0.25))
    Weight=1
    Child=RigVMVariableNode() {variable=calf_twist_01_l}.Value
    Parents.0.Item=RigVMVariableNode() {variable=foot_l}.Value
    Parents.1.Item=RigVMVariableNode() {variable=calf_twist_01_l}.Value
   )
   n17 = &RigVMFunction_MathQuaternionToEuler(
    RotationOrder=ZYX
    Value=&RigUnit_GetTransform(Space=LocalSpace, bInitial=False, Item=RigVMVariableNode() {variable=calf_twist_01_l}.Value).Transform.Rotation
   )
   > Set Transform = &RigUnit_SetRotation(
    Weight=1
    bInitial=False
    Value=&RigVMFunction_MathQuaternionFromEuler(Euler=(X=0,Y=0,Z=0), Euler.X=@n17.Result.X).Result
    Item=RigVMVariableNode() {variable=calf_twist_01_l}.Value
   )
   > Set Transform_1 = &RigUnit_SetRotation(
    Weight=1
    bInitial=False
    Value=&RigVMFunction_MathQuaternionFromEuler(
     Euler=(X=0,Y=0,Z=0)
     Euler.X=&RigVMFunction_MathFloatMul(B=0.5, A=@n17.Result.X) {template="Multiply::Execute(in A,in B,out Result)"}.Result
    ).Result
    Item=RigVMVariableNode() {variable=calf_twist_02_l}.Value
   )
   > &RigUnit_ParentConstraint(
    AdvancedSettings=(InterpolationType=Average,RotationOrderForFilter=XZY)
    Filter=(TranslationFilter=(bX=false,bY=false,bZ=false),RotationFilter=(bX=True,bY=false,bZ=false),ScaleFilter=(bX=false,bY=false,bZ=false))
    Parents=((Item=(Type=Bone,Name="foot_r"),Weight=0.75),(Item=(Type=Bone,Name="calf_twist_01_r"),Weight=0.25))
    Weight=1
    Child=RigVMVariableNode() {variable=calf_twist_01_r}.Value
    Parents.0.Item=RigVMVariableNode() {variable=foot_r}.Value
    Parents.1.Item=RigVMVariableNode() {variable=calf_twist_01_r}.Value
   )
   n18 = &RigVMFunction_MathQuaternionToEuler(
    RotationOrder=ZYX
    Value=&RigUnit_GetTransform(Space=LocalSpace, bInitial=False, Item=RigVMVariableNode() {variable=calf_twist_01_r}.Value).Transform.Rotation
   )
   > Set Transform_2 = &RigUnit_SetRotation(
    Weight=1
    bInitial=False
    Value=&RigVMFunction_MathQuaternionFromEuler(Euler=(X=0,Y=0,Z=0), Euler.X=@n18.Result.X).Result
    Item=RigVMVariableNode() {variable=calf_twist_01_r}.Value
   )
   > Set Transform_2_1 = &RigUnit_SetRotation(
    Weight=1
    bInitial=False
    Value=&RigVMFunction_MathQuaternionFromEuler(
     Euler=(X=0,Y=0,Z=0)
     Euler.X=&RigVMFunction_MathFloatMul(B=0.5, A=@n18.Result.X) {template="Multiply::Execute(in A,in B,out Result)"}.Result
    ).Result
    Item=RigVMVariableNode() {variable=calf_twist_02_r}.Value
   )
   > /Script/ControlRig.RigUnit_DebugHierarchy(
    Color=(R=1,G=1,B=1,A=1)
    DebugDrawSettings=(DepthPriority=SDPG_Foreground,Lifetime=-1)
    Items=()
    Scale=5
    Thickness=0.2
    WorldOffset=(Rotation=(X=0,Y=0,Z=0,W=1),Translation=(X=0,Y=0,Z=0),Scale3D=(X=1,Y=1,Z=1))
    bEnabled=false
   ) {is=struct RigVMUnitNode}
   > RigVMFunctionReturnNode()
  ]}
 RigVMFunctionLibrary {nodes [DeformRig_v1_Twists = RigVMCollapseNode(
    ExecuteContext=()
    calf_twist_01_l=(Type=None,Name="None")
    calf_twist_01_r=(Type=None,Name="None")
    calf_twist_02_l=(Type=None,Name="None")
    calf_twist_02_r=(Type=None,Name="None")
    foot_l=(Type=None,Name="None")
    foot_r=(Type=None,Name="None")
    hand_l=(Type=None,Name="None")
    hand_r=(Type=None,Name="None")
    lowerarm_twist_01_l=(Type=None,Name="None")
    lowerarm_twist_01_r=(Type=None,Name="None")
    lowerarm_twist_02_l=(Type=None,Name="None")
    lowerarm_twist_02_r=(Type=None,Name="None")
    thigh_l=(Type=None,Name="None")
    thigh_r=(Type=None,Name="None")
    thigh_twist_01_l=(Type=None,Name="None")
    thigh_twist_01_r=(Type=None,Name="None")
    upperarm_l=(Type=None,Name="None")
    upperarm_r=(Type=None,Name="None")
    upperarm_twist_01_l=(Type=None,Name="None")
    upperarm_twist_01_r=(Type=None,Name="None")
    upperarm_twist_02_l=(Type=None,Name="None")
    upperarm_twist_02_r=(Type=None,Name="None")
   )]}
 RigVMModel {nodes [RigUnit_BeginExecution = &RigUnit_BeginExecution()]}
]
