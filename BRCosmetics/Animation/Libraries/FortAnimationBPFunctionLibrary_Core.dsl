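; An animation Blueprint function library with helpers for constant rotation and material-driven eye blinks.
; CreateConstantRotationFunction returns the input rotator advanced per axis by rate times delta time.
; RemapMaterialBlinks copies left and right blink, upper-lid and inner-squint curves from an anim instance to a dynamic material.
; It also returns each eye's strongest closing value, with squint weighted, remapped to animation time in seconds, plus raw squints.
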(fn CreateConstantRotationFunction
  (params (RotationRate :type Rotator :default "0.000000,2.000000,0.000000") (RotationVar :type Rotator%20%28by%20ref%29) (DeltaTime :type Float%20%28double-precision%29 :default 0.03))
  (return (Math|Rotator|MakeRotator (+ (* DeltaTime (.roll RotationRate)) (.roll RotationVar)) (+ (* DeltaTime (.pitch RotationRate)) (.pitch RotationVar)) (+ (* DeltaTime (.yaw RotationRate)) (.yaw RotationVar)))))

(fn RemapMaterialBlinks
  (params (Character%20Part%20Anim%20Instance :type Anim%20Instance%20Object%20Reference) (Dynamic%20Material%20Instance :type Material%20Instance%20Dynamic%20Object%20Reference) (Squint%20Influence :type Float%20%28double-precision%29 :default 0.35) (Anim%20Length%28Seconds%29 :type Float%20%28double-precision%29 :default 3.3))
  (bind _character_part_anim_instance (Variables|GetCharacter%20Part%20Anim%20Instance))
  (bind _returnvalue (Animation|Curves|GetCurveValue _character_part_anim_instance "L_squint_inner_pose"))
  (bind _returnvalue_2 (Animation|Curves|GetCurveValue _character_part_anim_instance "R_squint_inner_pose"))
  (bind _returnvalue_3 (Animation|Curves|GetCurveValue _character_part_anim_instance "l_upperLidDown_pose"))
  (bind _returnvalue_4 (Animation|Curves|GetCurveValue _character_part_anim_instance "r_blink_pose"))
  (bind _returnvalue_5 (Animation|Curves|GetCurveValue _character_part_anim_instance "l_blink_pose"))
  (bind _returnvalue_6 (Animation|Curves|GetCurveValue _character_part_anim_instance "R_upperLidDown_pose"))
  (bind _dynamic_material_instance (Variables|GetDynamic%20Material%20Instance))
  ; Set the material params
  (Rendering|Material|SetScalarParameterValue _dynamic_material_instance "l_blink_pose" _returnvalue_5 :!class /Script/Engine.MaterialInstanceDynamic)
  (Rendering|Material|SetScalarParameterValue _dynamic_material_instance "r_blink_pose" _returnvalue_4 :!class /Script/Engine.MaterialInstanceDynamic)
  (Rendering|Material|SetScalarParameterValue _dynamic_material_instance "L_upperLidDown_pose" _returnvalue_3 :!class /Script/Engine.MaterialInstanceDynamic)
  (Rendering|Material|SetScalarParameterValue _dynamic_material_instance "R_upperLidDown_pose" _returnvalue_6 :!class /Script/Engine.MaterialInstanceDynamic)
  (Rendering|Material|SetScalarParameterValue _dynamic_material_instance "L_squint_inner_pose" _returnvalue :!class /Script/Engine.MaterialInstanceDynamic)
  (Rendering|Material|SetScalarParameterValue _dynamic_material_instance "R_squint_inner_pose" _returnvalue_2 :!class /Script/Engine.MaterialInstanceDynamic)
  (return (Math|Float|MapRangeClamped (Math|Float|Max(Float) _returnvalue_5 _returnvalue_3 (* _returnvalue (Variables|GetSquint%20Influence))) 0.0 1.0 0.0 (Variables|GetAnim%20Length%28Seconds%29)) (Math|Float|MapRangeClamped (Math|Float|Max(Float) _returnvalue_4 _returnvalue_6 (* _returnvalue_2 (Variables|GetSquint%20Influence))) 0.0 1.0 0.0 (Variables|GetAnim%20Length%28Seconds%29)) _returnvalue _returnvalue_2))
